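# Physics asset for the Dragon Racer female medium body that drives secondary motion of the jacket.
# Seven two-segment chains of box bodies hang from spine_04 with limited swing and twist, stiffest at the back and looser at the sides and front.
# Angular-locked constraints link neighboring lower segments around the back, leaving the front open, so the panels move together.
# Capsules on pelvis, spine and thighs provide body shapes for the jacket.
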PhysicsAsset /BRCosmetics/Characters/Player/Female/Medium/Bodies/F_MED_Dragon_Racer/Meshes/F_MED_Dragon_Racer_Dyn
bodies {
 dyn_jacket_bk_01 [{shapeType=Box, extentX=5, extentY=11.01, extentZ=2.5, center=<2.5 0 0>}]
 dyn_jacket_bk_02 [{shapeType=Box, extentX=5, extentY=11.01, extentZ=2.5, center=<2.5 0 0>}]
 dyn_jacket_bk_l_01 [{shapeType=Box, extentX=5, extentY=10, extentZ=2.5, center=<2.5 0 -1>, rotation=<rot 0 0 20>}]
 dyn_jacket_bk_l_02 [{shapeType=Box, extentX=5, extentY=10, extentZ=2.5, center=<2.5 0 -1>, rotation=<rot 0 0 20>}]
 dyn_jacket_bk_r_01 [{shapeType=Box, extentX=5, extentY=10, extentZ=2.5, center=<2.5 0 -1>, rotation=<rot 0 0 -20>}]
 dyn_jacket_bk_r_02 [{shapeType=Box, extentX=5, extentY=10, extentZ=2.5, center=<2.5 0 -1>, rotation=<rot 0 0 -20>}]
 dyn_jacket_fr_l_01 [{shapeType=Box, extentX=6, extentY=7.5, extentZ=2, center=<2.5 -3.5 -1>, rotation=<rot 0 0 -25>}]
 dyn_jacket_fr_l_02 [{shapeType=Box, extentX=6, extentY=7.5, extentZ=2, center=<2.5 -3.5 -1>, rotation=<rot 0 0 -25>}]
 dyn_jacket_fr_r_01 [{shapeType=Box, extentX=6, extentY=7.5, extentZ=2, center=<2.5 3.5 -1>, rotation=<rot 0 0 25>}]
 dyn_jacket_fr_r_02 [{shapeType=Box, extentX=6, extentY=7.5, extentZ=2, center=<2.5 3.5 -1>, rotation=<rot 0 0 25>}]
 dyn_jacket_l_01 [{shapeType=Box, extentX=5, extentY=8, extentZ=2, center=<2.5 0 0>, rotation=<rot 0 0 10>}]
 dyn_jacket_l_02 [{shapeType=Box, extentX=5, extentY=8, extentZ=2, center=<2.5 0 0>, rotation=<rot 0 0 10>}]
 dyn_jacket_r_01 [{shapeType=Box, extentX=5, extentY=8, extentZ=2, center=<2.5 0 0>, rotation=<rot 0 0 -10>}]
 dyn_jacket_r_02 [{shapeType=Box, extentX=5, extentY=8, extentZ=2, center=<2.5 0 0>, rotation=<rot 0 0 -10>}]
 pelvis [{shapeType=Capsule, length=10.3035, radius=10.3795, center=<-0.1856 -1.5184 -0.1972>, rotation=<rot 5.5723 -90 0.5389>}]
 spine_01 [{shapeType=Capsule, length=6.948, radius=8.7038, center=<-1.0334 1.1831 0.0196>, rotation=<rot 0.7761 -89.9999 -0.0409>}]
 spine_02 [{shapeType=Capsule, length=3.1, radius=8}]
 spine_04 [{shapeType=Capsule, length=3.5375, radius=8.1075, center=<-3.0072 2.9175 0.1822>, rotation=<rot -4.8294 -90 -0.5872>}]
 thigh_l [{shapeType=Capsule, length=16.4627, radius=8.0068, center=<-14.6231 0.3316 -0.4294>, rotation=<rot 0.0656 -92.9029 -90.7503>}]
 thigh_r [{shapeType=Capsule, length=16.2127, radius=8.2568, center=<14.6425 -0.1506 0.2624>, rotation=<rot 0.0656 -92.9028 -90.7494>}]
}
constraints [
 {bone1Name=dyn_jacket_fr_l_01, bone2Name=spine_04, swing1LimitDegrees=20, swing2LimitDegrees=20, twistLimitDegrees=5}
 {bone1Name=dyn_jacket_fr_l_02, bone2Name=dyn_jacket_fr_l_01, swing1LimitDegrees=30, swing2LimitDegrees=15, twistLimitDegrees=5}
 {bone1Name=dyn_jacket_fr_r_01, bone2Name=spine_04, swing1LimitDegrees=20, swing2LimitDegrees=20, twistLimitDegrees=5}
 {bone1Name=dyn_jacket_fr_r_02, bone2Name=dyn_jacket_fr_r_01, swing1LimitDegrees=30, swing2LimitDegrees=15, twistLimitDegrees=5}
 {bone1Name=dyn_jacket_l_01, bone2Name=spine_04, swing1LimitDegrees=15, swing2LimitDegrees=15, twistLimitDegrees=5}
 {bone1Name=dyn_jacket_l_02, bone2Name=dyn_jacket_l_01, swing1LimitDegrees=40, swing2LimitDegrees=15, twistLimitDegrees=5}
 {bone1Name=dyn_jacket_r_01, bone2Name=spine_04, swing1LimitDegrees=15, swing2LimitDegrees=15, twistLimitDegrees=5}
 {bone1Name=dyn_jacket_r_02, bone2Name=dyn_jacket_r_01, swing1LimitDegrees=40, swing2LimitDegrees=15, twistLimitDegrees=5}
 {bone1Name=dyn_jacket_bk_l_01, bone2Name=spine_04, swing1LimitDegrees=5, swing2LimitDegrees=5, twistLimitDegrees=5}
 {bone1Name=dyn_jacket_bk_l_02, bone2Name=dyn_jacket_bk_l_01, swing1LimitDegrees=20, swing2LimitDegrees=5, twistLimitDegrees=5}
 {bone1Name=dyn_jacket_bk_r_01, bone2Name=spine_04, swing1LimitDegrees=5, swing2LimitDegrees=5, twistLimitDegrees=5}
 {bone1Name=dyn_jacket_bk_r_02, bone2Name=dyn_jacket_bk_r_01, swing1LimitDegrees=20, swing2LimitDegrees=5, twistLimitDegrees=5}
 {bone1Name=dyn_jacket_bk_01, bone2Name=spine_04, swing1LimitDegrees=5, swing2LimitDegrees=5, twistLimitDegrees=5}
 {bone1Name=dyn_jacket_bk_02, bone2Name=dyn_jacket_bk_01, swing1LimitDegrees=5, swing2LimitDegrees=5, twistLimitDegrees=5}
 {
  bone1Name=dyn_jacket_l_02
  bone2Name=dyn_jacket_fr_l_02
  swing1LimitDegrees=45
  swing1Motion=Locked
  swing2LimitDegrees=45
  swing2Motion=Locked
  twistLimitDegrees=45
  twistMotion=Locked
 }
 {
  bone1Name=dyn_jacket_bk_l_02
  bone2Name=dyn_jacket_l_02
  swing1LimitDegrees=45
  swing1Motion=Locked
  swing2LimitDegrees=45
  swing2Motion=Locked
  twistLimitDegrees=45
  twistMotion=Locked
 }
 {
  bone1Name=dyn_jacket_bk_02
  bone2Name=dyn_jacket_bk_l_02
  swing1LimitDegrees=45
  swing1Motion=Locked
  swing2LimitDegrees=45
  swing2Motion=Locked
  twistLimitDegrees=45
  twistMotion=Locked
 }
 {
  bone1Name=dyn_jacket_r_02
  bone2Name=dyn_jacket_fr_r_02
  swing1LimitDegrees=45
  swing1Motion=Locked
  swing2LimitDegrees=45
  swing2Motion=Locked
  twistLimitDegrees=45
  twistMotion=Locked
 }
 {
  bone1Name=dyn_jacket_bk_r_02
  bone2Name=dyn_jacket_r_02
  swing1LimitDegrees=45
  swing1Motion=Locked
  swing2LimitDegrees=45
  swing2Motion=Locked
  twistLimitDegrees=45
  twistMotion=Locked
 }
 {
  bone1Name=dyn_jacket_bk_02
  bone2Name=dyn_jacket_bk_r_02
  swing1LimitDegrees=45
  swing1Motion=Locked
  swing2LimitDegrees=45
  swing2Motion=Locked
  twistLimitDegrees=45
  twistMotion=Locked
 }
]
