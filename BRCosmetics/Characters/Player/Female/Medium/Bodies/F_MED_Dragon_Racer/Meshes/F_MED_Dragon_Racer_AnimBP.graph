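# Animation Blueprint for the female Dragon Racer body cosmetic, copying the attached parent mesh's pose and adding procedural secondary motion.
# The event graph maps each thigh's yaw delta from reference pose into alphas for additive glute corrective offsets.
# Weighted thigh and twist-bone constraints position gun bones, with a Look At aiming dyn_gun.
# A RigidBody simulation in spine_04 space follows, then a fixed scale corrects the necklace neck bone.
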AnimBlueprint /BRCosmetics/Characters/Player/Female/Medium/Bodies/F_MED_Dragon_Racer/Meshes/F_MED_Dragon_Racer_AnimBP
target_skeleton=&/F_MED_Dragon_Racer_Skeleton
graphs [
 AnimGraph {nodes [
   n1 = "Transform (Modify) Bone - Bone: dyn_glute_l"(
    ComponentPose=Local To Component(LocalPose=Copy Pose From Mesh() {bUseAttachedParent=true})
    Alpha=Get thigh_rot_up_l() -> (thigh_rot_up_l:double)
    Translation="-6,-6.5,0.000"
    Rotation="0,0,0"
    Scale="1,1,1"
   ) {
    translationMode=BMM_Additive
    translationSpace=BCS_BoneSpace
    alphaBoolBlend {blendOption=Linear}
    boneToModify {boneName=dyn_glute_l}
    translation=<-6 -6.5>
   }
   n2 = "Transform (Modify) Bone - Bone: dyn_glute_r"(ComponentPose=@n1, Alpha=Get thigh_rot_up_r() -> (thigh_rot_up_r:double), Translation="-5,-6.5,0", Rotation="0,0,0", Scale="1,1,1") {
    translationMode=BMM_Additive
    translationSpace=BCS_BoneSpace
    alphaBoolBlend {blendOption=Linear}
    boneToModify {boneName=dyn_glute_r}
    translation=<-5 -6.5>
   }
   n3 = "Constraint - dyn_gun_MSTR (PP)"!(ComponentPose=@n2, Alpha=1, ConstraintWeights_0=0.507, ConstraintWeights_1:float=0.493) -> (Pose) {
    alphaBoolBlend {blendOption=Linear}
    boneToModify {boneName=dyn_gun_MSTR}
    constraintSetup [{targetBone {boneName=thigh_r}}, {targetBone {boneName=thigh_twist_01_r}}]
    constraintWeights [1,1]
   }
   n4 = "Constraint - dyn_gun_aim (PP)"!(ComponentPose=@n3, Alpha=1, ConstraintWeights_0=0.789, ConstraintWeights_1:float=0.211) -> (Pose) {
    alphaBoolBlend {blendOption=Linear}
    boneToModify {boneName=dyn_gun_aim}
    constraintSetup [{targetBone {boneName=thigh_r}}, {targetBone {boneName=thigh_twist_01_r}}]
    constraintWeights [1,1]
   }
   n5 = Look At - Bone: dyn_gun(ComponentPose=@n4, Alpha=1) {
    alphaBoolBlend {blendOption=Linear}
    boneToModify {boneName=dyn_gun}
    lookAtLocation {x=0}
    lookAtTarget {boneReference {boneName=dyn_gun_aim}}
    lookAt_Axis {axis=<1 0>}
   }
   n6 = RigidBody(ComponentPose=@n5) {
    bClampLinearTranslationLimitToRefPose=true
    bOverrideWorldGravity=true
    overridePhysicsAsset=&/F_MED_Dragon_Racer_Dyn
    simulationSpace=BaseBoneSpace
    alphaBoolBlend {blendOption=Linear}
    baseBoneRef {boneName=spine_04}
    componentAppliedLinearAccClamp=<2000 2000 2000>
    componentLinearAccScale=<0.5 0.5 0.5>
    componentLinearVelScale=<0.2 0.2 0.2>
    overrideWorldGravity {z=-980}
   }
   n7 = "Transform (Modify) Bone - Bone: neck_necklacefix"(ComponentPose=@n6, Alpha=1, Translation="0,0,0", Rotation="0,0,0", Scale="1,1.15,1") {
    rotationMode=BMM_Additive
    rotationSpace=BCS_BoneSpace
    scaleMode=BMM_Replace
    scaleSpace=BCS_BoneSpace
    translationMode=BMM_Additive
    translationSpace=BCS_BoneSpace
    alphaBoolBlend {blendOption=Linear}
    boneToModify {boneName=neck_necklacefix}
    scale {y=1.15}
   }
   Output Pose(Result=Component To Local(ComponentPose=@n7)) {layerGroup=None}
  ]}
 EventGraph {nodes [
   Event BlueprintUpdateAnimation()
   n1 = BreakTransform(InTransform=GetDeltaTransformFromRefPose(self=GetOwnerSkeletalMeshComponent(), BoneName=thigh_r, BaseName=pelvis))
   > Set thigh_rot_up_r(thigh_rot_up_r:double=MapRangeClamped(Value=BreakRotator(InRot=@n1.Rotation).Yaw, InRangeA=0, InRangeB=-100)) -> (Output_Get:double)
   n2 = BreakTransform(InTransform=GetDeltaTransformFromRefPose(self=GetOwnerSkeletalMeshComponent(), BoneName=thigh_l, BaseName=pelvis))
   > Set thigh_rot_up_l(thigh_rot_up_l:double=MapRangeClamped(Value=BreakRotator(InRot=@n2.Rotation).Yaw, InRangeA=-30, InRangeB=-100)) -> (Output_Get:double)
  ]}
]
variables [
 thigh_rot_up_l {type=number}
 thigh_rot_up_r {type=number}
]
