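# Physics asset for the dynamic hair on the face accessory part of the female medium Dragon Racer player cosmetic body.
# Twelve hair bones each carry one capsule body aligned along the bone, alongside a small capsule on the head bone.
# Every hair body constrains directly to the head rather than chaining to another, so each swings independently under identical asymmetric swing limits and a tight twist limit.
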PhysicsAsset /BRCosmetics/Characters/Player/Female/Medium/Bodies/F_MED_Dragon_Racer/Meshes/Parts/F_MED_Dragon_Racer_FaceAcc_Dyn_hair
bodies {
 dyn_hair_01 [{shapeType=Capsule, length=2.7616, radius=0.8983, center=<1.1817 0 0>, rotation=<rot 90 0 0>}]
 dyn_hair_02 [{shapeType=Capsule, length=4, radius=1, center=<1.7531 0 0>, rotation=<rot 90 0 0>}]
 dyn_hair_03 [{shapeType=Capsule, length=1.6394, radius=1.1491, center=<1.0026 -0.0001 0>, rotation=<rot 90 0 0>}]
 dyn_hair_04 [{shapeType=Capsule, length=2.103, radius=0.7363, center=<1.1855 0 0>, rotation=<rot 90 0 0>}]
 dyn_hair_05 [{shapeType=Capsule, length=3.5334, radius=1.4666, center=<1.7344 0 0.0003>, rotation=<rot 90 0 0>}]
 dyn_hair_06 [{shapeType=Capsule, length=2.3031, radius=1, center=<1.1368 0 -0.0001>, rotation=<rot 90 0 0>}]
 dyn_hair_07 [{shapeType=Capsule, length=3.7209, radius=1.2791, center=<1.4428 0 0.0001>, rotation=<rot 90 0 0>}]
 dyn_hair_08 [{shapeType=Capsule, length=2.7661, radius=1.4597, center=<1.0739 0 0>, rotation=<rot 90 0 0>}]
 dyn_hair_09 [{shapeType=Capsule, length=2.4481, radius=1, center=<0.5604 0 -0.0001>, rotation=<rot 90 0 0>}]
 dyn_hair_10 [{shapeType=Capsule, length=2.8632, radius=1, center=<1.4331 0 0.0001>, rotation=<rot 90 0 0>}]
 dyn_hair_11 [{shapeType=Capsule, length=3.3307, radius=1, center=<1.7095 0 0.0001>, rotation=<rot 90 0 0>}]
 dyn_hair_12 [{shapeType=Capsule, length=2.5156, radius=1, center=<1.3349 0 -0.0001>, rotation=<rot 90 0 0>}]
 head [{shapeType=Capsule, length=1, radius=1}]
}
constraints [
 {bone1Name=dyn_hair_01, bone2Name=head, swing1LimitDegrees=40, swing2LimitDegrees=25, twistLimitDegrees=5}
 {bone1Name=dyn_hair_02, bone2Name=head, swing1LimitDegrees=40, swing2LimitDegrees=25, twistLimitDegrees=5}
 {bone1Name=dyn_hair_03, bone2Name=head, swing1LimitDegrees=40, swing2LimitDegrees=25, twistLimitDegrees=5}
 {bone1Name=dyn_hair_04, bone2Name=head, swing1LimitDegrees=40, swing2LimitDegrees=25, twistLimitDegrees=5}
 {bone1Name=dyn_hair_05, bone2Name=head, swing1LimitDegrees=40, swing2LimitDegrees=25, twistLimitDegrees=5}
 {bone1Name=dyn_hair_06, bone2Name=head, swing1LimitDegrees=40, swing2LimitDegrees=25, twistLimitDegrees=5}
 {bone1Name=dyn_hair_07, bone2Name=head, swing1LimitDegrees=40, swing2LimitDegrees=25, twistLimitDegrees=5}
 {bone1Name=dyn_hair_08, bone2Name=head, swing1LimitDegrees=40, swing2LimitDegrees=25, twistLimitDegrees=5}
 {bone1Name=dyn_hair_09, bone2Name=head, swing1LimitDegrees=40, swing2LimitDegrees=25, twistLimitDegrees=5}
 {bone1Name=dyn_hair_10, bone2Name=head, swing1LimitDegrees=40, swing2LimitDegrees=25, twistLimitDegrees=5}
 {bone1Name=dyn_hair_11, bone2Name=head, swing1LimitDegrees=40, swing2LimitDegrees=25, twistLimitDegrees=5}
 {bone1Name=dyn_hair_12, bone2Name=head, swing1LimitDegrees=40, swing2LimitDegrees=25, twistLimitDegrees=5}
]
